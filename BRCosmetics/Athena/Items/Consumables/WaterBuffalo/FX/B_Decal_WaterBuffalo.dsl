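; Decal Blueprint for the WaterBuffalo consumable's FX that displays a splash decal and animates its fade-in.
; The construction script creates a dynamic instance of the splash decal material, stores it and assigns it to the Decal component.
; On BeginPlay, a timeline plays from the start and drives the material's FadeIn parameter.
; A DEBUG flag sets FadeIn to full during construction.
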(event EventBeginPlay
  (bind (_direction _newtrack_0) (Timeline|Timeline_0 PlayFromStart)
    (:Update
      (Rendering|Material|SetScalarParameterValue (Variables|Default|GetDecal_MID) "FadeIn" _newtrack_0 :!class /Script/Engine.MaterialInstanceDynamic))))

(fn ConstructionScript
  (bind _returnvalue (Rendering|Material|CreateDynamicMaterialInstance "/BRCosmetics/Athena/Items/Consumables/WaterBuffalo/FX/MD_Decal_WaterBuffalo_Splash.MD_Decal_WaterBuffalo_Splash" "None" "None" :!class /Script/Engine.KismetMaterialLibrary))
  (bind _output_get (Variables|Default|SetDecal_MID _returnvalue))
  (Rendering|Components|Decal|SetDecalMaterial (Variables|Decal|GetDecal) _output_get)
  (if (Variables|Default|GetDEBUG)
    (Rendering|Material|SetScalarParameterValue (Variables|Default|GetDecal_MID) "FadeIn" 1.0 :!class /Script/Engine.MaterialInstanceDynamic)))
