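# Translucent deferred decal material for the WaterBuffalo consumable's water splash effect.
# MF_DecalSphereCoords supplies distorted UVs for offset samples of splash and bubble textures, with the splash layer weighted by decal orientation.
# Erode_Alpha breaks up their sum using FadeIn and decal lifetime opacity, yielding the scaled opacity and a mask for skylight cubemap reflection.
# Shading path switches give mobile explicit mip 0 samples.
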Material /BRCosmetics/Athena/Items/Consumables/WaterBuffalo/FX/MD_Decal_WaterBuffalo_Splash
graph {
 properties {
  bAutomaticallySetUsageInEditor=false
  blendMode=BLEND_Translucent
  materialDomain=MD_DeferredDecal
  pixelDepthOffsetMode=PDOM_Legacy
  usedShadingModels=Default Lit
 }
 nodes [
  Constant() {r=0.5}
  n1 = ReflectionVectorWS()
  n2 = TextureSampleParameterCube!(UVs=@n1, Apply View MipBias) {parameterName=CubeMap, samplerType=SAMPLERTYPE_LinearColor, texture=/Game/Packages/Fortress_Sky/TexturesHDR/T_AthenaSkylight}
  n3 = TextureSampleParameterCube!(UVs=@n1, Level, Apply View MipBias) {
   constMipValue=0
   mipValueMode=TMVM_MipLevel
   parameterName=CubeMapMobile
   samplerType=SAMPLERTYPE_LinearColor
   texture=&T_AthenaSkylight
  }
  n4 = Multiply(A=ScalarParameter() {parameterName=FadeIn}, B=DecalLifetimeOpacity().Opacity)
  n5 = MaterialFunctionCall!(UVs=TextureCoordinate(), Additional Scale=ScalarParameter() {parameterName=Distortion}) {materialFunction=/Game/Effects/Fort_Effects/Materials/_Functions/MF_DecalSphereCoords}
  n6 = Multiply(A=@n5.Result) {constB=0.5}
  n7 = TextureSample(UVs=@n6) {
   samplerSource=SSM_Clamp_WorldGroupSettings
   samplerType=SAMPLERTYPE_LinearColor
   texture=/Game/Effects/Fort_Effects/Textures/Water/T_waterSplash
  }
  n8 = TextureSample!(UVs=@n6, Tex, Level, Apply View MipBias) {
   constMipValue=0
   mipValueMode=TMVM_MipLevel
   samplerSource=SSM_Clamp_WorldGroupSettings
   samplerType=SAMPLERTYPE_LinearColor
   texture=&T_waterSplash
  }
  n9 = Add(A=Multiply(A=@n5.Result) {constB=0.5}, B=Constant2Vector() {r=0.5})
  n10 = TextureSample!(UVs=@n9, Tex, Apply View MipBias) {
   samplerSource=SSM_Clamp_WorldGroupSettings
   samplerType=SAMPLERTYPE_LinearColor
   texture=&T_waterSplash
  }
  n11 = TextureSample!(UVs=@n9, Tex, Level, Apply View MipBias) {
   constMipValue=0
   mipValueMode=TMVM_MipLevel
   samplerSource=SSM_Clamp_WorldGroupSettings
   samplerType=SAMPLERTYPE_LinearColor
   texture=&T_waterSplash
  }
  n12 = Add(
   A=ShadingPathSwitch(Default=@n7.G, Mobile=@n8.G) {desc=Migrated from FeatureLevelSwitch}
   B=ShadingPathSwitch(Default=@n10.G, Mobile=@n11.G) {desc=Migrated from FeatureLevelSwitch}
  )
  n13 = DotProduct(A=ObjectOrientation(), B=Constant3Vector() {constant=<rgba 0 0 1 0>})
  n14 = Add(A=Multiply(A=@n5.Result) {constB=0.5}, B=Constant2Vector() {g=-0.5})
  n15 = ShadingPathSwitch(
   Default=TextureSample!(UVs=@n14, Tex, Apply View MipBias) {texture=/Game/Effects/Fort_Effects/Textures/Water/T_BoilBubble_D}.G
   Mobile=TextureSample!(UVs=@n14, Tex, Level, Apply View MipBias) {constMipValue=0, mipValueMode=TMVM_MipLevel, texture=&T_BoilBubble_D}.G
  ) {desc=Migrated from FeatureLevelSwitch}
  n16 = Add(A=Multiply(A=@n5.Result) {constB=0.5}, B=Constant2Vector() {g=-0.5, r=0.5})
  n17 = ShadingPathSwitch(
   Default=TextureSample!(UVs=@n16, Tex, Apply View MipBias) {texture=&T_BoilBubble_D}.G
   Mobile=TextureSample!(UVs=@n16, Tex, Level, Apply View MipBias) {constMipValue=0, mipValueMode=TMVM_MipLevel, texture=&T_BoilBubble_D}.G
  ) {desc=Migrated from FeatureLevelSwitch}
  n18 = MaterialFunctionCall!(
   Dynamic Param=Multiply(A=OneMinus(None=@n4), B=ScalarParameter() {defaultValue=-5, parameterName=ErodeStrength})
   Alpha=Add(A=Add(A=@n15, B=@n17), B=LinearInterpolate(B=@n12, Alpha=@n13))
  ) {materialFunction=/Game/Effects/Fort_Effects/Materials/_Functions/Erode_Alpha}
  LinearInterpolate(
   B=Multiply(
    A=ShadingPathSwitch(Default=@n2.RGB, Mobile=@n3.RGB) {desc=Migrated from FeatureLevelSwitch}
    B=ScalarParameter() {defaultValue=0.1, parameterName=Reflection INtensity}
   )
   Alpha=Saturate(None=Subtract(A=@n18.Result) {constB=3})
  )
  Multiply(A=@n18.Result, B=ScalarParameter() {defaultValue=0.15, parameterName=Master Opacity})
  ScalarParameter() {parameterName=Metallic}
  ScalarParameter() {defaultValue=1, parameterName=Roughness}
  VectorParameter() {parameterName=DecalColor, defaultValue=<rgba 0 0 0 1>}
 ]
}
